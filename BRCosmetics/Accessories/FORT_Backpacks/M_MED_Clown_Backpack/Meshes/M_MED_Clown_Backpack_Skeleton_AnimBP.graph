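# Animation Blueprint for the clown backpack cosmetic, adding secondary motion to its balloons.
# Each update maps spine yaw to backpack root rotation and derives a Perlin-noise force from pawn height while skydiving, damped under parachute.
# From ref pose, that rotation feeds spring AnimDynamics on three balloons, then a Trail, Look At and bone-relative AnimDynamics on balloon A1, all noise-driven except Look At.
# The aim-offset balloon-leg node sits outside the output chain.
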AnimBlueprint /BRCosmetics/Accessories/FORT_Backpacks/M_MED_Clown_Backpack/Meshes/M_MED_Clown_Backpack_Skeleton_AnimBP
target_skeleton=&/M_MED_Clown_Backpack_Skeleton
graphs [
 AnimGraph {nodes [
   Apply Limits!(ComponentPose:Component Space Pose Link Structure, Alpha:float=1, AngularOffsets_0:Vector="0, 0, 0") -> (Pose:Component Space Pose Link Structure) {
    is=title AnimGraphNode_ApplyLimits
    alphaBoolBlend {blendOption=Linear}
    angularOffsets [<0 0 0>]
    angularRangeLimits [{bone {boneName=None}, limitMax=<180 180 180>, limitMin=<-180 -180 -180>}]
   }
   Copy Bone - Source Bone: backpack_root - Target Bone: None!(ComponentPose:Component Space Pose Link Structure, Alpha:float=1) -> (Pose:Component Space Pose Link Structure) {is=title AnimGraphNode_CopyBone, alphaBoolBlend {blendOption=Linear}, sourceBone {boneName=backpack_root}}
   "Transform (Modify) Bone - Bone: dyn_balloonLeg_base"!(ComponentPose, Alpha=1, Rotation=Rotation From X Vector!(InVec:Vector=Get AimOffset() -> (AimOffset:Vector)) -> (ReturnValue:Rotator)) -> (Pose) {
    rotationMode=BMM_Additive
    rotationSpace=BCS_ParentBoneSpace
    alphaBoolBlend {blendOption=Linear}
    boneToModify {boneName=dyn_balloonLeg_base}
   }
   n1 = "Transform (Modify) Bone - Bone: backpack_root"(
    ComponentPose=Local To Component(LocalPose=Local Space Ref Pose())
    Alpha=1
    Translation="0,0,0"
    Rotation=Get backpack_rotate()
    Scale="1,1,1"
   ) {rotationMode=BMM_Additive, rotationSpace=BCS_BoneSpace, alphaBoolBlend {blendOption=Linear}, boneToModify {boneName=backpack_root}}
   n2 = Get falling_anim_dyn_up_force_vec3() -> (falling_anim_dyn_up_force_vec3:Vector)
   n3 = Anim Dynamics - Bone: dyn_balloonB1(ComponentPose=@n1, Alpha=1, ExternalForce=vector * float(A=@n2, B=0.5)) {
    angularDampingOverride=0.95
    angularSpringConstant=650
    bAngularSpring=true
    bGravityOverrideInSimSpace=true
    bLinearSpring=true
    bOverrideAngularDamping=true
    bOverrideLinearDamping=true
    bUsePlanarLimit=false
    gravityScale=0.01
    linearDampingOverride=0.99
    linearSpringConstant=650
    alphaBoolBlend {blendOption=Linear}
    boundBone {boneName=dyn_balloonB1}
    componentAppliedLinearAccClamp=<500 500 500>
    componentLinearAccScale=<1 1 1>
    componentLinearVelScale=<1 1 1>
    physicsBodyDefinitions [{
      sphereCollisionRadius=0
      boundBone {boneName=dyn_balloonB1}
      boxExtents=<5 5 5>
      constraintSetup {angularLimitsMax=<0 0 0>, angularLimitsMin=<0 0 0>, angularTarget=<0 0 0>, linearAxesMax=<0.1 1 0>, linearAxesMin=<0 -1 0>}
      localJointOffset=<0 0 -3>
     }]
   }
   n4 = Anim Dynamics - Bone: dyn_balloonC1(ComponentPose=@n3, Alpha=1, ExternalForce=vector * float(A=@n2, B=-0.5)) {
    angularDampingOverride=0.95
    angularSpringConstant=350
    bAngularSpring=true
    bGravityOverrideInSimSpace=true
    bLinearSpring=true
    bOverrideAngularDamping=true
    bOverrideLinearDamping=true
    bUsePlanarLimit=false
    gravityScale=0.01
    linearDampingOverride=0.99
    linearSpringConstant=350
    alphaBoolBlend {blendOption=Linear}
    boundBone {boneName=dyn_balloonC1}
    componentAppliedLinearAccClamp=<500 500 500>
    componentLinearAccScale=<1 1 1>
    componentLinearVelScale=<1 1 1>
    physicsBodyDefinitions [{
      sphereCollisionRadius=0
      boundBone {boneName=dyn_balloonC1}
      boxExtents=<5 5 5>
      constraintSetup {angularLimitsMax=<0 0 0>, angularLimitsMin=<0 0 0>, angularTarget=<0 0 0>, linearAxesMax=<0 1 0>, linearAxesMin=<-0.1 -1 0>}
      localJointOffset=<0 0 -3>
     }]
   }
   n5 = Anim Dynamics - Bone: dyn_balloonD1(ComponentPose=@n4, Alpha=1, ExternalForce=vector * float(A=@n2, B=-0.5)) {
    angularDampingOverride=0.95
    angularSpringConstant=500
    bAngularSpring=true
    bGravityOverrideInSimSpace=true
    bLinearSpring=true
    bOverrideAngularDamping=true
    bOverrideLinearDamping=true
    bUsePlanarLimit=false
    gravityScale=0.01
    linearDampingOverride=0.99
    linearSpringConstant=500
    alphaBoolBlend {blendOption=Linear}
    boundBone {boneName=dyn_balloonD1}
    componentAppliedLinearAccClamp=<500 500 500>
    componentLinearAccScale=<1 1 1>
    componentLinearVelScale=<1 1 1>
    physicsBodyDefinitions [{
      sphereCollisionRadius=0
      boundBone {boneName=dyn_balloonD1}
      boxExtents=<5 5 5>
      constraintSetup {angularLimitsMax=<0 0 0>, angularLimitsMin=<0 0 0>, angularTarget=<0 0 0>, linearAxesMax=<0 1 0>, linearAxesMin=<-0.1 -1 0>}
      localJointOffset=<0 0 -3>
     }]
    planarLimits [{drivingBone {boneName=backpack_root}, planeTransform {location=<0 5 0>, rotation=<rot 0 0 22>}}]
   }
   n6 = Get bIsSkyDiving()
   n7 = NormalizeToRange!(Value:double=BreakVector(InVec=@n2).Z, RangeMin:double=-1, RangeMax:double=1) -> (ReturnValue:double)
   n8 = Trail controller - Bone: dyn_balloonA1_01!(
    ComponentPose=@n5
    Alpha=float * float(A=SelectFloat(A=-1, B=1, bPickA=@n6), B=float * float(A=@n7, B=1) {is=title K2Node_CommutativeAssociativeBinaryOperator}) {is=title K2Node_PromotableOperator}
    FakeVelocity:Vector=vector * float(A=@n2, B=0.02)
   ) -> (Pose) {
    bUsePlanarLimit=true
    chainBoneAxis=Z
    relaxationSpeedScale=2
    baseJoint {boneName=dyn_balloonA1_base}
    planarLimits [
     {drivingBone {boneName=dyn_balloonA1_base}, planeTransform {location=<3 0 3>, rotation=<rot -90 22.0004 0>, scale=<1 1 1>}}
     {drivingBone {boneName=backpack_root}, planeTransform {location=<20 -1 0>, rotation=<rot 0 5.0001 -90>, scale=<1 1 1>}}
    ]
    rotationLimits [{limitMax=<180 180 180>}, {limitMax=<180 180 180>}]
    rotationOffsets [<0 0 0>, <0 0 0>]
    trailBone {boneName=dyn_balloonA1_01}
    trailRelaxationSpeed {editorCurveData {keys [<at 0 10 RCIM_Linear>,<at 1 8 RCIM_Linear>]}}
   }
   n9 = Look At - Bone: dyn_balloonA1_aim(ComponentPose=@n8, Alpha=1) {
    alphaBoolBlend {blendOption=Linear}
    boneToModify {boneName=dyn_balloonA1_aim}
    lookAtLocation {x=0}
    lookAtTarget {boneReference {boneName=dyn_balloonA1_base}}
    lookAt_Axis {axis=<-1 0>}
    lookUp_Axis {axis=<0 1>}
   }
   n10 = Anim Dynamics - Bone: dyn_balloonA1_aim(ComponentPose=@n9, Alpha=1, ExternalForce=vector * float(A=@n2, B=-0.005)) {
    angularDampingOverride=0.95
    angularSpringConstant=355
    bAngularSpring=true
    bGravityOverrideInSimSpace=true
    bLinearSpring=true
    bOverrideAngularDamping=true
    bOverrideLinearDamping=true
    bUsePlanarLimit=false
    gravityScale=0.001
    linearDampingOverride=0.99
    linearSpringConstant=355
    simulationSpace=BoneRelative
    alphaBoolBlend {blendOption=Linear}
    boundBone {boneName=dyn_balloonA1_aim}
    componentAppliedLinearAccClamp=<100 100 100>
    componentLinearAccScale=<1 1 1>
    componentLinearVelScale=<1 1 1>
    physicsBodyDefinitions [{
      sphereCollisionRadius=0
      boundBone {boneName=dyn_balloonA1_aim}
      boxExtents=<1 1 1>
      constraintSetup {angularLimitsMax=<0 0 0>, angularLimitsMin=<0 0 0>, angularTarget=<0 0 0>, linearAxesMax=<3 3 3>, linearAxesMin=<0 -1 -3>}
      localJointOffset=<-1 0 0>
     }]
    relativeSpaceBone {boneName=dyn_balloonA1_01}
   }
   Output Pose(Result=Component To Local(ComponentPose=@n10)) {layerGroup=None}
   Get pawnYaw() -> (pawnYaw:Rotator)
  ]}
 EventGraph {nodes [
   Event BlueprintUpdateAnimation()
   n1 = GetOwnerSkeletalMeshComponent()
   n2 = BreakRotator(InRot=BreakTransform(InTransform=GetDeltaTransformFromRefPose(self=@n1, BoneName=spine_05, BaseName=spine_01)).Rotation)
   > Set backpack_rotate(backpack_rotate=MakeRotator(Yaw=MapRangeClamped(Value=@n2.Yaw, InRangeA=-5, InRangeB=-30, OutRangeB=25)))
   n3 = Get PawnLastLocation() -> (PawnLastLocation:Vector)
   n4 = SelectFloat(
    A=Make Perlin Noise and Remap!(Value:float=BreakVector(InVec=@n3).Z, RangeOutMin:float=-25000, RangeOutMax:float=25000) -> (ReturnValue:float)
    B=0
    bPickA=Get bIsSkyDiving()
   )
   n5 = SelectFloat(A=float * float(A=@n4, B=0.1), B=@n4, bPickA=Get bIsParachuteOpened())
   > Set falling_anim_dyn_up_force_vec3(falling_anim_dyn_up_force_vec3:Vector=vector * float(A=MakeVector(Y=@n5), B=0.75)) -> (Output_Get:Vector)
   n6 = float * float(A=Get AimPitch(), B=Get AimOffsetMultiplier() -> (AimOffsetMultiplier:double))
   > Set!(execute, AimOffset_X:double=@n6, AimOffset_Y:double=0, AimOffset_Z:double=0) -> (then, Output_Get:Vector)
   > Set Is Sky Diving(Is Sky Diving:double="To Float (Boolean)"(InBool=Get bIsSkyDiving())) -> (Output_Get:double)
   "Absolute (Float)"(A=0)
   TryGetPawnOwner()
   "Min (Float)"(A:double=@n6, B:double=0) -> (ReturnValue:double) {is=title K2Node_CommutativeAssociativeBinaryOperator}
   float + float(A=BreakRotator(InRot=Get PawnLastRotation() -> (PawnLastRotation:Rotator)).Yaw, B=0) {is=title K2Node_CommutativeAssociativeBinaryOperator}
   n7 = Set aimingPitch(aimingPitch:double=Get AimPitch()) -> (Output_Get:double)
   Set pawnYaw(pawnYaw:Rotator=MakeRotator(Pitch=float * float(A=@n7, B=-1))) -> (Output_Get:Rotator)
  ]}
]
variables [
 aimingPitch {type=number}
 AimOffset {type=Vector}
 AimOffsetMultiplier {type=number, default=-1}
 backpack_rotate {type=Rotator}
 falling_anim_dyn_up_force_vec3 {type=Vector}
 Is Sky Diving {type=number}
 Negate Is Sky Diving {type=number}
 pawnYaw {type=Rotator}
]
